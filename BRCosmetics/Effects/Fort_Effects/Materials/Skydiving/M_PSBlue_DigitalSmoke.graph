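# Translucent unlit particle material for skydiving effects, rendering smoke as a glowing hologram with blue default tints and pink edge lines.
# Normal-map Fresnel and animated scanlines blend glow colors, while subtracting two offset normal-map samples produces blueprint-like edges.
# A rotating flowmap distorts two smoke textures, with dynamic parameters driving distortion, softness blend and erosion.
# That mask multiplies hologram opacity alongside depth, near-camera and distance fades, with mobile skipping depth fade.
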Material /BRCosmetics/Effects/Fort_Effects/Materials/Skydiving/M_PSBlue_DigitalSmoke
graph {
 properties {
  bUsedWithNiagaraSprites=true
  bUsedWithParticleSprites=true
  blendMode=BLEND_Translucent
  pixelDepthOffsetMode=PDOM_Legacy
  shadingModel=MSM_Unlit
  usedShadingModels=Unlit
 }
 nodes [
  n1 = ScalarParameter() {defaultValue=0.8, parameterName=Base Color Bias}
  n2 = TextureSampleParameter2D() {group=Textures, parameterName=Normal Map, samplerType=SAMPLERTYPE_Normal, texture=/Game/Effects/Fort_Effects/Textures/Smoke/SmokeTest_N}
  n3 = MaterialFunctionCall(
   Brightness=ScalarParameter() {defaultValue=0.4, parameterName=Fresnel Brightness}
   Exponent=ScalarParameter() {defaultValue=3, parameterName=Fresnel Exponent}
   Vertex / Pixel Normal=Transform(None=@n2.RGB)
  ) {materialFunction=&MF_Fresnel_FX}
  n4 = ScalarParameter() {defaultValue=1, parameterName=Secondary Line Brightness}
  n5 = MaterialFunctionCall!(
   Number of Scanlines=ScalarParameter() {defaultValue=150, parameterName=Scanline Amount}
   Scanline Speed
   "Scanlines Thickness (Bias)"=ScalarParameter() {defaultValue=-0.15, parameterName=ScanlineSize}
   Scanlines Dark Value=ScalarParameter() {defaultValue=0.5, parameterName=Scanlines Dark}
   Scanlines Light Value=ScalarParameter() {defaultValue=1, parameterName=Scanlines Light}
   Block Opacity=MaterialFunctionCall!(WPO Intensity A, WPO Intensity B, Time, Bias=Constant()) {materialFunction=/Game/Effects/Fort_Effects/Effects/Atlas/Materials/MF_Atlas_HologramWiggle_01}.Intense 0-1
   UVs
   Time
   Block Repeat
  ) {materialFunction=/Game/Effects/Fort_Effects/Effects/Atlas/Materials/MF_Atlas_Scanlines_01}
  n6 = LinearInterpolate(
   A=LinearInterpolate(B=VectorParameter() {parameterName=Glow Color Inner, defaultValue=<rgba 0 0.1349 1 1>}.RGB, Alpha=Clamp(None=@n1)) {constA=1}
   B=VectorParameter() {parameterName=Glow Color Outer, defaultValue=<rgba 0 1.3492 10 1>}.RGB
   Alpha=Add(A=@n3.Fresnel Inverted, B=Multiply(A=@n4, B=@n5.Secondary Lines))
  )
  n7 = VectorParameter() {parameterName=Glow Color Lines, defaultValue=<rgba 7 0 5.6656 1>}
  n8 = ScalarParameter() {defaultValue=0.0075, parameterName=Offset U}
  n9 = ScalarParameter() {defaultValue=0.0006, parameterName=Offset V}
  n10 = TextureSampleParameter2D(UVs=Add(A=AppendVector(A=@n8, B=@n9), B=TextureCoordinate())) {group=Textures, parameterName=Normal Map, samplerType=SAMPLERTYPE_Normal, texture=&SmokeTest_N}
  n11 = TextureSampleParameter2D() {group=Textures, parameterName=Normal Map, samplerType=SAMPLERTYPE_Normal, texture=&SmokeTest_N}
  n12 = Subtract(A=@n10.RGB, B=@n11.RGB)
  n13 = Constant() {r=1}
  n14 = Power(Base=Saturate(None=DotProduct(A=@n12, B=@n13)), Exp=ScalarParameter() {defaultValue=1.5, parameterName=Blueprint Edge Exponent})
  n15 = ParticleColor()
  MaterialFunctionCall(Emissive=Multiply(A=LinearInterpolate(A=@n6, B=@n7.RGB, Alpha=@n14), B=@n15.RGB)) {materialFunction=&MF_FX_EmissiveAdjustments}
  n16 = ScalarParameter() {defaultValue=0.7, parameterName=Opacity Bias Scale}
  n17 = Multiply(A=@n3.Fresnel Inverted, B=@n16)
  n18 = OneMinus(None=@n16)
  n19 = ScalarParameter() {defaultValue=1.5, parameterName=Blueprint Edge Brightness}
  n20 = Multiply(A=Add(A=Add(A=@n17, B=@n18), B=Multiply(A=@n14, B=@n19)), B=@n5.Result)
  n21 = ScalarParameter() {defaultValue=0.2, parameterName=Secondary Line Opacity}
  n22 = TextureSample(UVs=Rotator(Coordinate=TextureCoordinate()) {centerX=0.25, centerY=0.25, speed=0.5}) {samplerType=SAMPLERTYPE_LinearColor, texture=/Game/Global/Textures/Default/Flowmaps/T_SwirlFlow_02}
  n23 = Add(
   A=TextureCoordinate()
   B=Multiply(
    A=ComponentMask(None=@n22.RGB) {g=true, r=true}
    B=DynamicParameter() {defaultValue=<rgba 1 0.06 1 1>, paramNames [Erode,UVdistort,Softness,Param4]}.UVdistort
   )
  )
  n24 = TextureSample(UVs=@n23) {samplerType=SAMPLERTYPE_LinearColor, texture=/Game/Effects/Fort_Effects/Textures/Smoke/T_SmokeSwirl}
  n25 = TextureSample(UVs=@n23) {samplerType=SAMPLERTYPE_LinearColor, texture=/Game/Effects/Fort_Effects/Textures/Smoke/SmokeTest2_Soft}
  n26 = DynamicParameter() {defaultValue=<rgba 1 0.06 1 1>, paramNames [Erode,UVdistort,Softness,Param4]}
  n27 = Multiply(A=Power(Base=LinearInterpolate(A=@n24.G, B=@n25.G, Alpha=@n26.Softness), Exp=@n26.Erode), B=ParticleColor().A)
  n28 = ScalarParameter() {defaultValue=500, parameterName=ShadowDistance}
  n29 = ShadingPathSwitch(Default=DepthFade(Opacity=@n27, FadeDistance=Multiply(A=@n28) {constB=0.8}), Mobile=@n27) {desc=Migrated from FeatureLevelSwitch}
  n30 = MaterialFunctionCall(Distance=LinearInterpolate(Alpha=@n27) {constA=360, constB=180}) {materialFunction=&Near_Camera_Fade}
  n31 = MaterialFunctionCall(
   Fade Length=ScalarParameter() {defaultValue=750, parameterName="Fade Length (S)"}
   Fade Offset=ScalarParameter() {defaultValue=3000, parameterName="Fade Offset (S)"}
  ) {materialFunction=&CameraDepthFade}
  Multiply(
   A=Add(A=Multiply(A=@n15.A, B=@n20), B=Multiply(A=@n5.Secondary Lines, B=@n21))
   B=Multiply(A=Multiply(A=@n29, B=@n30.Result), B=OneMinus(None=@n31.Result))
  )
 ]
}
