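# Additive, unlit, two-sided material for PlayStation-themed skydiving ribbon and beam trails.
# Local-space trail UVs, scrolled by a front-end motion parameter, sample soft and thin smoke textures for a mask combining particle alpha and near-camera fade.
# Inverted smoke indexes a Sony color curve, tinted by particle color and HDR-boosted by time of day.
# A sine/cosine chain builds spiral ribbon offsets from object position, with simulated movement in the front-end preview.
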Material /BRCosmetics/Effects/Fort_Effects/Materials/Skydiving/M_PlaystationFlow
graph {
 properties {
  bUsedWithBeamTrails=true
  bUsedWithNiagaraRibbons=true
  blendMode=BLEND_Additive
  pixelDepthOffsetMode=PDOM_Legacy
  refractionMethod=RM_IndexOfRefraction
  shadingModel=MSM_Unlit
  twoSided=true
  usedShadingModels=Unlit
 }
 nodes [
  n1 = ParticleColor()
  n2 = MaterialFunctionCall!(
   MaxiumURange=ScalarParameter() {defaultValue=1250, parameterName=Soft Radius}
   RangeRampHardness=ScalarParameter() {parameterName=Soft Hardness}
   MirrorTexCoordU
  ) {materialFunction=/BRCosmetics/Weapons/FORT_Melee/Pickaxe_Venus/MF_LocalSpaceTrailUV}
  n3 = ComponentMask(None=@n2."LocalSpaceCoordinate(UV)") {r=true}
  n4 = Multiply(A=Time(), B=DynamicParameter() {defaultValue=<rgba 0 1 1 0>, paramNames [FrontEndMotion, Param2, Param3, Param4]}.FrontEndMotion) {constB=0}
  n5 = ScalarParameter() {defaultValue=3, parameterName=U Scale}
  n6 = AppendVector(A=Multiply(A=Add(A=@n3, B=@n4), B=@n5), B=ComponentMask(None=@n2."LocalSpaceCoordinate(UV)") {g=true})
  n7 = TextureSample(UVs=@n6) {samplerType=SAMPLERTYPE_Grayscale, texture=&T_Smoke_Trail_Soft}
  n8 = MaterialFunctionCall(
   MaxiumURange=ScalarParameter() {defaultValue=2500, parameterName=Stylized Radius}
   RangeRampHardness=ScalarParameter() {parameterName=Stylized Hardness}
  ) {materialFunction=&MF_LocalSpaceTrailUV}
  n9 = Multiply(A=@n5, B=Add(A=@n4, B=ComponentMask(None=@n8."LocalSpaceCoordinate(UV)") {r=true}))
  n10 = ComponentMask(None=@n8."LocalSpaceCoordinate(UV)") {g=true}
  n11 = Add(
   A=Power(Base=@n7.R, Exp=Constant() {r=4})
   B=TextureSample(UVs=AppendVector(A=@n9, B=@n10)) {texture=/Game/Effects/Fort_Effects/Textures/Smoke/T_Smoke_ThinTrail_Packed}.B
  )
  n12 = ScalarParameter() {defaultValue=220, parameterName=Near Camera Fade Distance}
  LinearInterpolate(Alpha=Multiply(
    A=Multiply(A=@n1.A, B=@n11)
    B=MaterialFunctionCall(Distance=@n12) {materialFunction=&Near_Camera_Fade}.Result
   )) {constA=1, constB=1.1}
  n13 = CurveAtlasRowParameter(CurveTime=OneMinus(None=@n7.R)) {
   atlas=&AthenaCurveAtlas01
   curve=/Game/Athena/LinearColorCurves/SonyColorCurve
   defaultValue=10
   parameterName=Sony Color
  }
  n14 = Multiply(A=@n13, B=@n1.RGB)
  n15 = ScalarParameter() {defaultValue=3, parameterName=Morning_HDR_Intensity}
  n16 = ScalarParameter() {defaultValue=3, parameterName=Day_HDR_Intensity}
  n17 = ScalarParameter() {defaultValue=2, parameterName=Evening_HDR_Intensity}
  n18 = ScalarParameter() {defaultValue=1, parameterName=Night_HDR_Intensity}
  MaterialFunctionCall(Emissive=MaterialFunctionCall(
    LDR Input=Clamp(None=@n14)
    HDR Tint=@n14
    HDR Intensity Multiplier=MaterialFunctionCall(Color_Morning=@n15, Color_Day=@n16, Color_Evening=@n17, Color_Night=@n18) {materialFunction=&Color_Time_of_Day}.Result
   ) {materialFunction=&DeriveHDRfromLDR}.RemappedRange) {materialFunction=&MF_FX_EmissiveAdjustments}
  n19 = MaterialFunctionCall() {materialFunction=&MF_LocalSpaceTrailUV}
  n20 = ObjectPositionWS()
  n21 = CollectionParameter() {collection=&FortniteMaterialParameters, parameterName=AthenaFrontendPreview}
  n22 = Multiply(A=Multiply(A=@n21, B=Time())) {constB=-2500}
  n23 = ComponentMask(None=Multiply(A=Add(A=@n20, B=@n22)) {constB=-1}) {b=true}
  n24 = ScalarParameter() {defaultValue=2500, parameterName=OffsetSpeed}
  n25 = Add(
   A=ComponentMask(None=@n19."LocalSpaceCoordinate(UV)") {r=true}
   B=Add(A=Divide(A=@n23, B=@n24) {constB=3500}, B=ScalarParameter() {parameterName=RandomOffset})
  )
  n26 = Multiply(A=AppendVector(A=Sine(None=@n25), B=Cosine(None=@n25))) {constB=75}
  n27 = OneMinus(None=@n19."LocalSpaceCoordinate(UV)")
  n28 = ComponentMask(None=Clamp(None=Subtract(A=@n27) {constB=0.2})) {r=true}
  n29 = ComponentMask(None=TextureCoordinate() {vTiling=2}) {g=true}
  n30 = Multiply(A=@n25)
  n31 = Cosine(None=@n30)
  n32 = Sine(None=@n30)
  n33 = Multiply(A=Multiply(A=Subtract(A=@n29), B=AppendVector(A=@n31, B=@n32)) {constB=8}, B=ScalarParameter() {defaultValue=30, parameterName=RibbonScale}) {constB=10}
  n34 = ComponentMask(None=@n27) {r=true}
  Multiply(A=Add(A=Multiply(A=@n26, B=@n28), B=Multiply(A=@n33, B=@n34)), B=ScalarParameter() {defaultValue=1, parameterName=WPOStrength}) {constB=0}
  Reroute(None=TextureSample(UVs=@n6) {samplerType=SAMPLERTYPE_Normal, texture=/Game/Effects/Fort_Effects/Textures/Smoke/T_Smoke_Trail_Soft_N}.RGB)
 ]
}
